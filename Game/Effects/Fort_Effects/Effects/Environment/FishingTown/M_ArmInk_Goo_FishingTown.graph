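# Masked, lit material for animated ink goo in the FishingTown environment effects.
# A panning swirl normal map distorts the goo texture's UVs and, flattened, blends into the surface normal by the goo's green channel.
# That channel also drives a powered purple emissive tint and a temporally dithered mask.
# An inverted sphere mask on the player position from the material parameter collection cuts the goo away near the player.
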Material /Game/Effects/Fort_Effects/Effects/Environment/FishingTown/M_ArmInk_Goo_FishingTown
graph {
 properties {bAutomaticallySetUsageInEditor=false, blendMode=BLEND_Masked, usedShadingModels=Default Lit}
 nodes [
  CameraPositionWS()
  n1 = TextureCoordinate()
  n2 = ScalarParameter() {defaultValue=2, parameterName=PanningTile_Mult}
  n3 = ScalarParameter() {defaultValue=-0.01, parameterName=X_Speed}
  n4 = ScalarParameter() {parameterName=Y_Speed}
  n5 = TextureSampleParameter2D(UVs=Panner(Coordinate=Multiply(A=@n1, B=@n2), Speed=AppendVector(A=@n3, B=@n4)) {speedX=-0.01}) {parameterName=DistortTexture, samplerType=SAMPLERTYPE_Normal, texture=/Game/VisualThreatManager/StormVisuals/LargeClouds/T_StormSwirlNormal}
  n6 = ScalarParameter() {defaultValue=0.05, parameterName=Distort_Strength}
  n7 = ComponentMask(None=@n5.RGB) {g=true, r=true}
  n8 = TextureSample(UVs=Add(A=TextureCoordinate(), B=Multiply(A=@n6, B=@n7))) {texture=&/T_InkArm_Goo}
  LinearInterpolate(
   A=VectorParameter() {parameterName=Normal, defaultValue=<rgba 0 0 1 0>}.RGB
   B=MaterialFunctionCall(Normal=@n5.RGB, Flatness=ScalarParameter() {defaultValue=0.75, parameterName=NormalFlatten}) {materialFunction=&FlattenNormal}.Result
   Alpha=@n8.G
  )
  n9 = ScalarParameter() {defaultValue=8, parameterName=Param_4}
  n10 = VectorParameter() {parameterName=Param_1, defaultValue=<rgba 0.0533 0.0076 0.1104 1>}
  n11 = VectorParameter() {defaultValue=<rgba 0.0367 0 0.0938 1>}
  Multiply(
   A=Multiply(A=Power(Base=@n8.G, Exp=@n9), B=LinearInterpolate(A=@n10.RGB, B=@n11.RGB, Alpha=@n8.G))
   B=ScalarParameter() {defaultValue=2, parameterName=Emmisive_Mult}
  )
  n12 = WorldPosition()
  n13 = CollectionParameter() {collection=&FortniteMaterialParameters, parameterName=PlayerPosition}
  n14 = ComponentMask(None=@n13) {b=true, g=true, r=true}
  n15 = ScalarParameter() {defaultValue=500, parameterName=SphereMask_Radius}
  n16 = ScalarParameter() {defaultValue=0.75, parameterName=SphereMask_Hardness}
  Multiply(
   A=OneMinus(None=SphereMask(A=@n12.XYZ, B=@n14, Radius=@n15, Hardness=@n16))
   B=Multiply(
    A=@n8.G
    B=MaterialFunctionCall(Alpha Threshold=@n8.G) {materialFunction=&DitherTemporalAA}.Result
   )
  )
  TransformPosition() {transformSourceType=TRANSFORMPOSSOURCE_World, transformType=TRANSFORMPOSSOURCE_World}
  Transform(None=MaterialFunctionCall!(Bump=@n8.G) {materialFunction=/Engine/Functions/Engine_MaterialFunctions03/Procedurals/PerturbNormalLQ}.WS Normal) {transformSourceType=TRANSFORMSOURCE_World, transformType=TRANSFORM_Tangent}
 ]
}
